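# Niagara system for the ScarletBionic backpack cosmetic, drawing glowing, flickering electric-style arcs on its skeletal mesh.
# AirGlow_and_Decal keeps three particles on mesh sockets, with waveform-flickered sprites and lights tinted by a system light color.
# Arcs repeatedly bursts short-lived ribbons between three socket pairs, shaped by Hermite splines using socket orientations as tangents.
# Random offsets, widths and UVs vary each arc, while curves fade them and confine sprites to the ribbon ends.
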NiagaraSystem /BRCosmetics/Accessories/FORT_Backpacks/Backpack_F_MED_ScarletBionic/FX/NS_Backpack_F_MED_ScarletBionic
dependencies {
 usedDataInterfaces [&NiagaraDataInterfaceArrayInt32, &NiagaraDataInterfaceCurve]
 usedDynamicInputs []
 usedModules [
  &self:SystemSpawnScript.SystemScriptSource.SystemScriptGraph.NiagaraNodeAssignment_1.SetVariables_#1
  &SystemState
 ]
}
emitters {
 AirGlow_and_Decal {
  emitterUpdateScript [
   &EmitterState(
    Inactive Response=::NewEnumerator1
    Life Cycle Mode=::NewEnumerator1
    Loop Behavior=::NewEnumerator1
    Loop Duration Mode=::NewEnumerator1
    UseLoopDelay=false
   )
   &SpawnBurst_Instantaneous(Spawn Count=3, Use Spawn Probability=false, UseLoopCountLimit=false)
  ]
  particleSpawnScript [
   &PartitionParticles(Number Of Partitions=3)
   &InitializeParticle(Color=$System.LightColor, Sprite Size Mode=::NewEnumerator3, Uniform Sprite Size=5, UsePositionOffset=false, Write Material Random=true)
   /SetVariables(Particles.LightExponent=3, Particles.LightRadius=5) {isSetParameters=true}
   /SetVariables(Particles.AirGlowSpriteScale:Vector2f=&Multiply_Vector2D_ByFloat(Float=25, Vector2D=<1 1>)) {isSetParameters=true}
   &CameraOffset(Camera Offset Amount=5)
  ]
  particleUpdateScript [
   &ParticleState(Kill Particles When Lifetime Has Elapsed=false)
   &SkeletalMeshLocation(
    Bone / Socket Index=$Particles.PartitionParticles.Partition
    MaskSpawnGroup=false
    Mesh Sampling Type=::NewEnumerator5
    Sampled Position Offset=<1.5 1 0>
    Skeletal Mesh=$User.SkeletalMesh
    Socket Sampling Mode=::NewEnumerator5
    Use Sampled Mesh Attribute as Ribbon ID=false
   )
   &SkeletalMeshLocation(
    Bone / Socket Index=$Particles.PartitionParticles.Partition
    MaskSpawnGroup=false
    Mesh Sampling Type=::NewEnumerator5
    Position Sampling=::NewEnumerator2
    Sampled Position Offset=<1 0 0>
    Skeletal Mesh=$User.SkeletalMesh
    Socket Sampling Mode=::NewEnumerator5
    Use Sampled Mesh Attribute as Ribbon ID=false
   ) {module=SkeletalMeshLocation001}
   /SetVariables(Particles.LightColor=&MultiplyLinearColor(
     Linear Color=$System.LightColor
     Scale Factor=&MakeVector4_FromFloat(Float=&Waveform(
       Global Input Position=&ReturnNormalizedExecIndex(Normalized Index Scale=20)
       Global Phase=&Add_Float(A=$Emitter.Age, B=&RandomRangeFloat(Minimum=0, Maximum=5))
       Global Amplitude Scale=200
       "Waveform [1]"=::NewEnumerator8
       "[1] Frequency"=6
       "[1] Amplitude Min/Max"=<0 1>
       "[1] Scale"=0.4
      ))
    )) {isSetParameters=true}
   /SetVariables(
    Particles.AirGlowColor:LinearColor=&MultiplyLinearColor(
     Linear Color=$System.LightColor
     Scale Factor=&MakeVector4_FromFloat(Float=&Waveform(
       Global Input Position=&ReturnNormalizedExecIndex(Normalized Index Scale=20)
       Global Phase=&Add_Float(A=$Emitter.Age, B=&RandomRangeFloat(Minimum=0, Maximum=5))
       Global Amplitude Scale=0.35
       "Waveform [1]"=::NewEnumerator8
       "[1] Frequency"=6
       "[1] Amplitude Min/Max"=<0 1>
       "[1] Scale"=0.4
       "[1] Bias"=0.7
      ))
    )
    Particles.AirGlowPosition:NiagaraPosition=$Particles.SkeletalMeshLocation001.SampledPosition
   ) {isSetParameters=true}
  ]
  rendererClasses [&NiagaraSpriteRendererProperties, &NiagaraLightRendererProperties]
  renderers [
   {rendererClass=&NiagaraSpriteRendererProperties}
   {rendererClass=&NiagaraLightRendererProperties, rendererIndex=1}
   {rendererClass=&NiagaraSpriteRendererProperties, rendererIndex=2}
  ]
 }
 Arcs {
  emitterSpawnScript [/SetVariables(Emitter.NumParticles:NiagaraInt32=&RandomRangeInt(Minimum=3, Maximum=6)) {isSetParameters=true}]
  emitterUpdateScript [
   &EmitterState(
    Life Cycle Mode=::NewEnumerator1
    Loop Duration=&RandomRangeFloat(Minimum=0.05, Maximum=0.2)
    UseLoopDelay=false
   )
   &SpawnBurst_Instantaneous(Spawn Count=$Emitter.NumParticles, Use Spawn Probability=false, UseLoopCountLimit=false)
   &SpawnBurst_Instantaneous(Spawn Count=$Emitter.NumParticles, Use Spawn Probability=false, UseLoopCountLimit=false) {module=SpawnBurst_Instantaneous002}
   &SpawnBurst_Instantaneous(Spawn Count=$Emitter.NumParticles, Use Spawn Probability=false, UseLoopCountLimit=false) {module=SpawnBurst_Instantaneous001}
   /SetVariables(Emitter.RibbonLife:NiagaraFloat=&RandomRangeFloat(Minimum=0.1, Maximum=0.5, Evaluation Type=::NewEnumerator1)) {isSetParameters=true}
   /SetVariables(Emitter.flip UV:NiagaraBool=&RandomBool(Evaluation Type=::NewEnumerator1)) {isSetParameters=true}
  ]
  particleSpawnScript [
   &InitializeParticle(
    Color=$System.LightColor
    Lifetime=$Emitter.RibbonLife
    Ribbon Width=&Multiply_Float(A=2, B=&RandomRangeFloat(Minimum=0.5, Maximum=1))
    Ribbon Width Mode=::NewEnumerator1
    Sprite Size Mode=::NewEnumerator3
    Uniform Sprite Size=1
    UsePositionOffset=false
    Write Material Random=true
   )
   /SetVariables(Particles.get flip UV:NiagaraBool=$Emitter.flip UV) {isSetParameters=true}
  ]
  particleUpdateScript [
   &ParticleState()
   /SetVariables(Particles.GetRandomRibbonOffset:NiagaraFloat=&self:ReNormalize()) {isSetParameters=true}
   &PartitionParticles(
    Normalized Index In Partition Includes One=true
    Particles Per Partition:NiagaraInt32=$Emitter.NumParticles
    Partition Count Mode:ENiagara_InfiniteLoopDuration=::NewEnumerator1
    Partition Mode:ENiagara_PartitionMode=::NewEnumerator1
    Write Partition Index to Ribbon Link Order:ENiagara_AttributeSamplingApplyOutput=::NewEnumerator0
    Write Partition to Ribbon ID=::NewEnumerator0
   )
   &SkeletalMeshLocation(
    Bone / Socket Index=&SelectIntFromArray(
     Direct Array Index=$Particles.PartitionParticles.Partition
     Direct Array Mode=::NewEnumerator1
     Int Selection Array=DataInterface() {propertyValues {IntData [0,1,2]}}
    )
    MaskSpawnGroup=false
    Mesh Sampling Type=::NewEnumerator5
    Position Sampling=::NewEnumerator2
    Sampled Position Offset=&Add_Vector(A=&Multiply_Vector(A=&RandomVector(Vector Scale=2), B=<0 1 1>), B=<0 0 0>)
    Skeletal Mesh=$User.SkeletalMesh
    Socket Sampling Mode=::NewEnumerator5
    Use Sampled Mesh Attribute as Ribbon ID=false
   )
   &SkeletalMeshLocation(
    Bone / Socket Index=&SelectIntFromArray(
     Direct Array Index=$Particles.PartitionParticles.Partition
     Direct Array Mode=::NewEnumerator1
     Int Selection Array=DataInterface() {propertyValues {IntData [3,4,5]}}
    )
    MaskSpawnGroup=false
    Mesh Sampling Type=::NewEnumerator5
    Position Sampling=::NewEnumerator2
    Sampled Position Offset=&Add_Vector(A=&Multiply_Vector(A=&RandomVector(Vector Scale=3), B=<0 1 1>), B=<0 0 0>)
    Skeletal Mesh=$User.SkeletalMesh
    Socket Sampling Mode=::NewEnumerator5
    Use Sampled Mesh Attribute as Ribbon ID=false
   ) {module=SkeletalMeshLocation001}
   /Game/Effects/Niagara/Modules/Curves/FN_HermiteSpline(
    Direction End:Vector3f=&ReturnMeshOrientationAxis(Mesh Orientation Axis=::XAxis, Mesh Orientation Quat=$Particles.SkeletalMeshLocation001.SampledRotation)
    Direction Start:Vector3f=&ReturnMeshOrientationAxis(Mesh Orientation Axis=::XAxis, Mesh Orientation Quat=$Particles.SkeletalMeshLocation.SampledRotation)
    Position Blend:NiagaraFloat=$Particles.PartitionParticles.Normalized Index In Partition
    Position End:Vector3f=$Particles.SkeletalMeshLocation001.SampledPosition
    Position Start:Vector3f=$Particles.SkeletalMeshLocation.SampledPosition
    Write Particles Position=true
   )
   &ScaleRibbonWidth(Ribbon Width Scale=&FloatFromCurve(FloatCurve=DataInterface() {propertyValues {Curve {keys [
         <at 0 0 RCIM_Cubic>
         {arriveTangentWeight=0.1067, interpMode=RCIM_Cubic, leaveTangentWeight=0.3202, tangentMode=RCTM_User, time=0.2, value=1}
         {arriveTangentWeight=0.3202, interpMode=RCIM_Cubic, leaveTangentWeight=0.1067, tangentMode=RCTM_User, time=0.8, value=1}
         <at 1 0 RCIM_Cubic>
        ]}}}))
   &ScaleColor(
    Scale Alpha=&FloatFromCurve(FloatCurve=DataInterface() {propertyValues {Curve {keys [
         {interpMode=RCIM_Cubic, tangentMode=RCTM_User, time=0, value=0}
         {interpMode=RCIM_Cubic, tangentMode=RCTM_User, time=0.5, value=1}
         {tangentMode=RCTM_User, time=1, value=0}
        ]}}})
    ScaleA=true
    ScaleRGB=true
   )
   /SetVariables(Particles.RibbonU1Override:NiagaraFloat=&Add_Float(A=$Particles.PartitionParticles.Normalized Index In Partition, B=$Particles.GetRandomRibbonOffset)) {isSetParameters=true}
   &ScaleSpriteSize(
    Scale Sprite Size Mode=::NewEnumerator0
    Uniform Scale Factor=&Multiply_Float(
     A=&FloatFromCurve(FloatCurve=DataInterface() {propertyValues {Curve {keys [
          {arriveTangent=5.6263, interpMode=RCIM_Cubic, leaveTangentWeight=0.9524, time=0, value=0}
          {
           arriveTangent=0.6967
           arriveTangentWeight=0.1114
           interpMode=RCIM_Cubic
           leaveTangent=-5.6928
           leaveTangentWeight=0.9633
           tangentMode=RCTM_Break
           tangentWeightMode=RCTWM_WeightedBoth
           time=0.7128
           value=1
          }
         ]}}})
     B=&FloatFromCurve(
      FloatCurve=DataInterface() {propertyValues {Curve {keys [<at 0 1>,<at 0.05 0>,<at 0.95 0>,<at 1 1>]}}}
      CurveIndex=$Particles.PartitionParticles.Normalized Index In Partition
     )
    )
   )
  ]
  rendererClasses [&NiagaraSpriteRendererProperties, &NiagaraRibbonRendererProperties]
  renderers [
   {rendererClass=&NiagaraSpriteRendererProperties}
   {rendererClass=&NiagaraRibbonRendererProperties, rendererIndex=1}
  ]
 }
}
summary {
 systemName=NS_Backpack_F_MED_ScarletBionic
 emitters [
  {emitterName=Arcs, rendererClasses [&NiagaraSpriteRendererProperties, &NiagaraRibbonRendererProperties]}
  {
   emitterName=AirGlow_and_Decal
   rendererClasses [&NiagaraSpriteRendererProperties, &NiagaraLightRendererProperties]
  }
 ]
 userVariables [
  User.bIsFrontEnd {type=NiagaraBool}
  User.bIsFrontEndPreview {type=NiagaraBool}
  User.SkeletalMesh {
   type=NiagaraDataInterfaceSkeletalMesh
   defaultValue {
    struct=NiagaraExt_VariableValue_DataInterface
    value {dataInterface=&self:NiagaraDataInterfaceSkeletalMesh_0}
   }
  }
 ]
}
