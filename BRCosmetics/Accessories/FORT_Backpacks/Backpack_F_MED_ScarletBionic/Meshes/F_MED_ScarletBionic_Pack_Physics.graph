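# Physics asset for the ScarletBionic cosmetic backpack mesh, defining its collision bodies and joints.
# A box covers the pack, a sphere covers dyn_pack_base, and three zero-length capsules, effectively offset spheres, cover the dyn_cone bones.
# The pack and base are joined with all angular motion locked.
# Each cone joins the base with swing locked and twist free, so it can only spin around its twist axis.
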PhysicsAsset /BRCosmetics/Accessories/FORT_Backpacks/Backpack_F_MED_ScarletBionic/Meshes/F_MED_ScarletBionic_Pack_Physics
bodies {
 dyn_cone_1 [{shapeType=Capsule, length=0, radius=5.75, shapeName=dyn_cone_1_capsule, center=<4.0022 0 0>, rotation=<rot 90 -179.9472 -179.9474>}]
 dyn_cone_2 [{shapeType=Capsule, length=0, radius=5.75, shapeName=dyn_cone_2_capsule, center=<4.0022 0 0>, rotation=<rot 90 179.9998 -180>}]
 dyn_cone_3 [{shapeType=Capsule, length=0, radius=5.75, shapeName=dyn_cone_3_capsule, center=<4.0022 0 0>, rotation=<rot 90 179.7468 179.7465>}]
 dyn_pack_base [{shapeType=Sphere, radius=8.5, shapeName=dyn_pack_base_sphere}]
 pack [{shapeType=Box, extentX=23.5, extentY=3, extentZ=36.5, shapeName=pack_box, center=<0 2.538 -4.3023>}]
}
constraints [
 {
  bone1Name=pack
  bone2Name=dyn_pack_base
  swing1LimitDegrees=45
  swing1Motion=Locked
  swing2LimitDegrees=45
  swing2Motion=Locked
  twistLimitDegrees=45
  twistMotion=Locked
 }
 {
  bone1Name=dyn_cone_1
  bone2Name=dyn_pack_base
  swing1LimitDegrees=45
  swing1Motion=Locked
  swing2LimitDegrees=45
  swing2Motion=Locked
  twistLimitDegrees=45
  twistMotion=Free
 }
 {
  bone1Name=dyn_cone_2
  bone2Name=dyn_pack_base
  swing1LimitDegrees=45
  swing1Motion=Locked
  swing2LimitDegrees=45
  swing2Motion=Locked
  twistLimitDegrees=45
  twistMotion=Free
 }
 {
  bone1Name=dyn_cone_3
  bone2Name=dyn_pack_base
  swing1LimitDegrees=45
  swing1Motion=Locked
  swing2LimitDegrees=45
  swing2Motion=Locked
  twistLimitDegrees=45
  twistMotion=Free
 }
]
