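# Animation Blueprint for the Scarlet Bionic backpack cosmetic, adding physics sway and an idle hover to the pack and its cone bones.
# AnimDynamics on the pack feeds a RigidBody node, then additive bone transforms driven by real-time sine waves.
# Each update runs crouch, sprint and skydive branches in series, so the skydive branch sets the final AnimDynamics scales and gravity.
# Crouching lowers RigidBody alpha, which is never restored.
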AnimBlueprint /BRCosmetics/Accessories/FORT_Backpacks/Backpack_F_MED_ScarletBionic/Meshes/F_MED_ScarletBionic_Pack_AnimBP
target_skeleton=&/F_MED_ScarletBionic_Pack_Skeleton
graphs [
 AnimGraph {nodes [
   n1 = Anim Dynamics - Bone: pack!(
    ComponentPose=Local To Component(LocalPose=Input Pose())
    Alpha=0.5
    GravityOverride:Vector=Get Gravity_Override() -> (Gravity_Override:Vector)
    ComponentLinearAccScale:Vector=Get Linear_Acc_Scale() -> (Linear_Acc_Scale:Vector)
    ComponentLinearVelScale:Vector=Get Linear_Vel_Scale() -> (Linear_Vel_Scale:Vector)
    ExternalForce="0,0,0"
   ) -> (Pose) {
    alpha=0
    angularBiasOverride=1
    angularDampingOverride=0.75
    angularSpringConstant=20
    bGravityOverrideInSimSpace=true
    bLinearSpring=true
    bOverrideLinearDamping=true
    bUseGravityOverride=true
    bUsePlanarLimit=false
    bUseSphericalLimits=true
    gravityScale=0
    linearDampingOverride=0.81
    linearSpringConstant=20
    simulationSpace=BoneRelative
    alphaBoolBlend {blendOption=Linear}
    boundBone {boneName=pack}
    componentAppliedLinearAccClamp=<50 65 130>
    componentLinearAccScale=<0.7 0.7 1.5>
    componentLinearVelScale=<0.7 0.7 1.5>
    gravityOverride {z=20}
    physicsBodyDefinitions [{
      sphereCollisionRadius=0
      boundBone {boneName=pack}
      boxExtents=<20 8 20>
      constraintSetup {
       linearXLimitType=Free
       linearZLimitType=Free
       angularLimitsMax=<0 0 0>
       angularLimitsMin=<0 0 0>
       angularTarget=<0 0 0>
       linearAxesMax=<2.5 0.5 0>
       linearAxesMin=<-2.5 -10 0>
      }
      localJointOffset=<0 0 0>
     }]
    relativeSpaceBone {boneName=root}
    sphericalLimits [{limitRadius=7, limitType=Inner, drivingBone {boneName=pack}, sphereLocalOffset=<0 0 4.4>}]
   }
   n2 = RigidBody(ComponentPose=@n1, Alpha=Get dyn_rigidBody_Alfa() -> (dyn_rigidBody_Alfa:double)) {
    bOverrideWorldGravity=true
    overlapChannel=ECC_PhysicsBody
    overridePhysicsAsset=&/F_MED_ScarletBionic_Pack_Physics
    alphaBoolBlend {blendOption=Linear}
    baseBoneRef {boneName=root}
    componentAppliedLinearAccClamp {y=10}
    componentLinearAccScale=<0.1 0.1 0.1>
    componentLinearVelScale=<0.1 0.1 0.1>
    simSpaceSettings {maxAngularAcceleration=0, maxAngularVelocity=0, maxLinearAcceleration=50, maxLinearVelocity=50, worldAlpha=0.5}
   }
   n3 = "Transform (Modify) Bone - Bone: pack"(ComponentPose=@n2, Translation=Get dyn_pack_Float() -> (dyn_pack_Float:Vector), Rotation="0,0,0") {translationMode=BMM_Additive, translationSpace=BCS_BoneSpace, alphaBoolBlend {blendOption=Linear}, boneToModify {boneName=pack}}
   n4 = "Transform (Modify) Bone - Bone: dyn_cone_1"(ComponentPose=@n3, Alpha=1, Translation=Get dyn_pack_Float_1() -> (dyn_pack_Float_1:Vector), Rotation="0,0,0", Scale="1,1,1") {translationMode=BMM_Additive, translationSpace=BCS_BoneSpace, alphaBoolBlend {blendOption=Linear}, boneToModify {boneName=dyn_cone_1}}
   n5 = "Transform (Modify) Bone - Bone: dyn_cone_2"(ComponentPose=@n4, Alpha=1, Translation=Get dyn_pack_Float_2() -> (dyn_pack_Float_2:Vector), Rotation="0,0,0", Scale="1,1,1") {translationMode=BMM_Additive, translationSpace=BCS_BoneSpace, alphaBoolBlend {blendOption=Linear}, boneToModify {boneName=dyn_cone_2}}
   n6 = "Transform (Modify) Bone - Bone: dyn_cone_3"(ComponentPose=@n5, Alpha=1, Translation=Get dyn_pack_Float_3() -> (dyn_pack_Float_3:Vector), Rotation="0,0,0", Scale="1,1,1") {
    rotationMode=BMM_Additive
    rotationSpace=BCS_BoneSpace
    translationMode=BMM_Additive
    translationSpace=BCS_BoneSpace
    alphaBoolBlend {blendOption=Linear}
    boneToModify {boneName=dyn_cone_3}
   }
   Output Pose(Result=Component To Local(ComponentPose=@n6)) {layerGroup=None}
  ]}
 EventGraph {nodes [
   Event BlueprintUpdateAnimation()
   > n1 = Sequence!(execute) -> (then_0, then_1, then_2:Exec, then_3:Exec, then_4:Exec, then_5:Exec)
   n2 = GetRealTimeSeconds()
   n3 = "Sin (Radians)"(A=float * float(A=float * float(A=1.27, B=1.21), B=@n2))
   > Set dyn_pack_Float_1(dyn_pack_Float_1:Vector=MakeVector(X=float * float(A=@n3, B=1.1), Y=1)) -> (Output_Get:Vector)
   n4 = Branch(execute=@n1.then_1, Condition=Get bIsCrouching())
   > Set Linear_Acc_Scale(Linear_Acc_Scale:Vector="0, 0, 0") -> (Output_Get:Vector)
   > Set Linear_Vel_Scale(Linear_Vel_Scale:Vector="0, 0, 0") -> (Output_Get:Vector)
   > n5 = Set dyn_rigidBody_Alfa(dyn_rigidBody_Alfa:double=0.25) -> (Output_Get:double)
   n6 = Get bIsCrouching()
   n7 = Branch(execute=[@n11.then, @n5.then], Condition=AND Boolean(A=Get bIsSprinting(), B=NOT Boolean(A=@n6)))
   > Set Linear_Acc_Scale(Linear_Acc_Scale="0.1,0.1,0.1")
   > n8 = Set Linear_Vel_Scale(Linear_Vel_Scale="9,11,13")
   n9 = Branch(execute=[@n8.then, @n10.then], Condition=Get bIsSkyDiving())
   > Set Linear_Acc_Scale(Linear_Acc_Scale="0.1,0.1,0.1")
   > Set Linear_Vel_Scale(Linear_Vel_Scale="1.2,1.2,0.6")
   > Set Gravity_Override(Gravity_Override:Vector="0,-1200,-980") -> (Output_Get:Vector)
   Set Linear_Acc_Scale(execute=@n9.else, Linear_Acc_Scale="0.1,0.1,0.1")
   > Set Linear_Vel_Scale(Linear_Vel_Scale="0,0,1.3")
   > Set Gravity_Override(Gravity_Override="0, 0, 0")
   Set Linear_Acc_Scale(execute=@n7.else, Linear_Acc_Scale="0.1,0.1,0.1")
   > Set Linear_Vel_Scale(Linear_Vel_Scale="9,250,13")
   > n10 = Set Gravity_Override(Gravity_Override="0, 0, 0")
   Set Linear_Acc_Scale(execute=@n4.else, Linear_Acc_Scale="0.1,0.1,0.1")
   > Set Linear_Vel_Scale(Linear_Vel_Scale="10,11,13")
   > n11 = Set Gravity_Override(Gravity_Override="0, 0, 0")
   n12 = GetRealTimeSeconds()
   n13 = "Sin (Radians)"(A=float * float(A=float * float(A=1.11, B=1.11), B=@n12))
   n14 = "Sin (Radians)"(A=float * float(A=float * float(A=1.24, B=1.24), B=@n12))
   n15 = "Sin (Radians)"(A=float * float(A=float * float(A=0.92, B=1.09), B=@n12))
   Set dyn_pack_Float(
    execute=@n1.then_2
    dyn_pack_Float:Vector=MakeVector(X=float * float(A=@n14, B=1.8), Y=float * float(A=@n13, B=0.9), Z=float * float(A=@n15, B=2.2))
   ) -> (Output_Get:Vector)
   n16 = GetRealTimeSeconds()
   n17 = "Sin (Radians)"(A=float * float(A=float * float(A=1.36, B=1.2), B=@n16))
   Set dyn_pack_Float_2(execute=@n1.then_3, dyn_pack_Float_2:Vector=MakeVector(X=float * float(A=@n17, B=1.1), Y=0)) -> (Output_Get:Vector)
   n18 = GetRealTimeSeconds()
   n19 = "Sin (Radians)"(A=float * float(A=float * float(A=1.36, B=0.75), B=@n18))
   Set dyn_pack_Float_3(execute=@n1.then_4, dyn_pack_Float_3:Vector=MakeVector(X=float * float(A=@n19, B=1.2), Y=0)) -> (Output_Get:Vector)
   TryGetPawnOwner()
   n20 = float * float(A=-0.85, B=1.25)
   float * float(A="Sin (Radians)"(A=float * float(A=@n20, B=@n18)), B=0.18)
   n21 = float * float(A=0.92, B=1.09)
   float * float(A="Sin (Radians)"(A=float * float(A=@n21, B=@n2)), B=0.2)
   n22 = float * float(A=1.5, B=1.11)
   float * float(A="Sin (Radians)"(A=float * float(A=@n22, B=@n2)), B=10)
   n23 = float * float(A=0.85, B=1.25)
   float * float(A="Sin (Radians)"(A=float * float(A=@n23, B=@n16)), B=1.9)
   n24 = float * float(A=1.11, B=1.11)
   float * float(A="Sin (Radians)"(A=float * float(A=@n24, B=@n16)), B=10)
   n25 = float * float(A=1.11, B=1.11)
   float * float(A="Sin (Radians)"(A=float * float(A=@n25, B=@n18)), B=50)
  ]}
]
variables [
 dyn_pack_Float {type=Vector}
 dyn_pack_Float_1 {type=Vector}
 dyn_pack_Float_2 {type=Vector}
 dyn_pack_Float_3 {type=Vector}
 dyn_pack_Rotate_Float_1 {type=Vector}
 dyn_rigidBody_Alfa {type=number, default=0.75}
 Gravity_Override {type=Vector}
 Linear_Acc_Scale {type=Vector}
 Linear_Vel_Scale {type=Vector}
]
